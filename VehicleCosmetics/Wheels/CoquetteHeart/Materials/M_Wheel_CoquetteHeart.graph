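# Skeletal-mesh material for the CoquetteHeart vehicle wheel, adding a glowing gem-like effect to part of the rim.
# MF_WheelParent_Simple supplies base attributes, overridden in the rim mask's blue channel except where vertex red restores them.
# The gem layers Fresnel color gradients, BumpOffset parallax normal highlights and reflection-vector sampling of a borrowed DuelingDragonsSE cubemap into base color and emissive.
# The result passes through vehicle damage and de-rez transition functions.
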Material /VehicleCosmetics/Wheels/CoquetteHeart/Materials/M_Wheel_CoquetteHeart
graph {
 properties {
  bAutomaticallySetUsageInEditor=false
  bUseMaterialAttributes=true
  bUsedWithMeshDeformer=true
  bUsedWithSkeletalMesh=true
  pixelDepthOffsetMode=PDOM_Legacy
  usedShadingModels=Default Lit
 }
 nodes [
  n1 = MaterialFunctionCall() {materialFunction=&MF_WheelParent_Simple}
  n2 = GetMaterialAttributes(None=@n1.MaterialAttributes) {attributeGetTypes [#1]}
  n3 = VectorParameter() {
   desc="RG: Scale\r\nBA: Offset"
   group=2_Rim Properties
   parameterName=RimNormalTransform
   channelNames {a=V-Offset, b=H-Offset, g=V-Scale, r=H-Scale}
   defaultValue=<rgba 1 1 0 0>
  }
  n4 = Multiply(A=TextureCoordinate(), B=AppendVector(A=@n3.H-Scale, B=@n3.H-Scale))
  n5 = AppendVector(A=@n3.H-Offset, B=@n3.H-Offset)
  n6 = LinearInterpolate(
   A=Constant3Vector() {constant=<rgba 0 0 1 0>}
   B=TextureSampleParameter2D(UVs=Add(A=@n4, B=@n5)) {parameterName=Rim_Normal, samplerType=SAMPLERTYPE_Normal, texture=/VehicleCosmetics/Textures/T_BlankNormal_B}.RGB
  )
  n7 = Transform(None=@n6)
  n8 = Saturate(None=ConstantBiasScale(None=Fresnel(Normal=@n7) {baseReflectFraction=0, exponent=0.3}) {bias=-0.1, scale=1.5})
  n9 = OneMinus(None=Fresnel(Normal=Transform(None=@n6)) {baseReflectFraction=0})
  n10 = LinearInterpolate(
   A=LinearInterpolate(
    A=VectorParameter() {group=2_Rim Properties, parameterName=Rim_ColorCAccent, defaultValue=<rgba 0.9 0 0.37 1>}.RGB
    B=@n2.Base Color
    Alpha=@n8
   )
   B=@n8
   Alpha=Saturate(None=ConstantBiasScale(None=@n9) {bias=0.02, scale=0.25})
  )
  n11 = VectorParameter() {
   desc="RG: Scale\r\nBA: Offset"
   group=2_Rim Properties
   parameterName=Rim_NormalDetailTransform
   channelNames {a=V-Offset, b=H-Offset, g=V-Scale, r=H-Scale}
   defaultValue=<rgba 20 10 0 0>
  }
  n12 = Multiply(A=TextureCoordinate(), B=AppendVector(A=@n11.H-Scale, B=@n11.H-Scale))
  n13 = AppendVector(A=@n11.H-Offset, B=@n11.H-Offset)
  n14 = VectorParameter() {
   group=2_Rim Properties
   parameterName=Rim_GemTransform
   channelNames {a=HeightRatio-Rim, b=Height-Rim, g=HeightRatio-Detail, r=Height-Detail}
   defaultValue=<rgba 0.2 -0.05 1 -0.02>
  }
  n15 = TextureSampleParameter2D(UVs=BumpOffset(Coordinate=Add(A=@n12, B=@n13), Height=@n14.Height-Detail, HeightRatioInput=@n14.Height-Detail)) {
   group=1_Input Textures
   parameterName=Rim_NormalDetail
   samplerType=SAMPLERTYPE_Normal
   texture=/VehicleCosmetics/Wheels/_SharedTextures/FX/Detail_Rocky_N
  }
  n16 = TextureSampleParameter2D(UVs=BumpOffset(Height=@n14.Height-Rim, HeightRatioInput=@n14.Height-Rim)) {parameterName=Rim_Normal, samplerType=SAMPLERTYPE_Normal, texture=&T_BlankNormal_B}
  n17 = Fresnel(Normal=Transform(None=LinearInterpolate(A=@n15.RGB, B=@n16.RGB) {constAlpha=0.9})) {baseReflectFraction=0, exponent=0.5}
  n18 = Saturate(None=ConstantBiasScale(None=OneMinus(None=@n17)) {bias=-0.7, scale=1.5})
  n19 = Transform(None=ReflectionVectorWS()) {transformSourceType=TRANSFORMSOURCE_Local}
  n20 = TextureObject() {samplerType=SAMPLERTYPE_LinearColor, texture=/VehicleCosmetics/Wheels/DuelingDragonsSE/Materials/TC_DuelingDragonsSE_GrassCube}
  n21 = Power(Base=Desaturation(
    None=TextureSample(UVs=@n19, Tex=@n20) {samplerType=SAMPLERTYPE_Normal, texture=/VehicleCosmetics/Wheels/CoquetteHeart/Textures/T_Rim_CoquetteHeart_N}.RGB
    Fraction=Constant()
   ) {luminanceFactors=<rgba 0.2126 0.7152 0.0722 1>})
  n22 = LinearInterpolate(A=Add(A=Add(A=@n10, B=@n18), B=@n21), B=Saturate(None=Add(B=@n21) {constA=1}), Alpha=@n18)
  n23 = Saturate(None=@n22)
  n24 = VectorParameter() {parameterName=Rim_ColorC_PBR, defaultValue=<rgba 0.5 0.04 1 1>}
  n25 = BlendMaterialAttributes(
   A=@n2.MaterialAttributes
   B=SetMaterialAttributes!(MaterialAttributes=@n2.MaterialAttributes, Base Color=@n22, Emissive Color=Multiply(A=@n23, B=@n24.R)) {attributeSetTypes [#1, #2]}
   Alpha=ComponentMask(None=@n1.RimRGBAMask) {b=true}
  )
  n26 = VertexColor()
  MaterialFunctionCall(In=MaterialFunctionCall(Material=BlendMaterialAttributes(A=@n25, B=@n2.MaterialAttributes, Alpha=@n26.R)) {materialFunction=&MF_BaseVehicle_Damage}.Result) {materialFunction=&MF_VehicleDeRez_Transition}
 ]
}
